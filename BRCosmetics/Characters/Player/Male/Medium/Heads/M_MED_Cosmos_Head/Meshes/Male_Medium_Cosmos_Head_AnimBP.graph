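# Animation Blueprint for the Cosmos head cosmetic on the male medium player character.
# It copies the pose from the attached parent mesh, then applies a slight, constant additive bone-space yaw to the head bone.
# Each update, it records whether the owner skeletal mesh's anim instance is playing the GuineaPig emote montage in isGuineaPigEmote, which the AnimGraph does not read.
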AnimBlueprint /BRCosmetics/Characters/Player/Male/Medium/Heads/M_MED_Cosmos_Head/Meshes/Male_Medium_Cosmos_Head_AnimBP
target_skeleton=&/Male_Medium_Cosmos_Head_Skeleton
graphs [
 AnimGraph {nodes [
   n1 = "Transform (Modify) Bone - Bone: head"!(ComponentPose=Local To Component(LocalPose=Copy Pose From Mesh() {bUseAttachedParent=true}), Alpha=1, Rotation="0,-5,0") -> (Pose) {
    rotationMode=BMM_Additive
    rotationSpace=BCS_BoneSpace
    alphaBoolBlend {blendOption=Linear}
    alphaScaleBiasClamp {bInterpResult=true, interpSpeedDecreasing=1, interpSpeedIncreasing=1}
    boneToModify {boneName=head}
    rotation {yaw=-5}
   }
   Output Pose(Result=Component To Local(ComponentPose=@n1)) {layerGroup=None}
  ]}
 EventGraph {nodes [
   Event BlueprintUpdateAnimation()
   n1 = GetAnimInstance(self=GetOwnerSkeletalMeshComponent())
   > Is Valid(InputObject=@n1)
   > Set isGuineaPigEmote(isGuineaPigEmote:bool=Montage_IsPlaying(self=@n1, Montage=/BRCosmetics/Animation/Game/MainPlayer/Emotes/GuineaPig/Emote_GuineaPig_CMM_FE_M)) -> (Output_Get:bool)
   TryGetPawnOwner()
  ]}
]
variables [isGuineaPigEmote {type=boolean}]
